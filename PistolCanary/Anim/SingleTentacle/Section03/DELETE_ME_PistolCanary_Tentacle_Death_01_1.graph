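# Level Sequence, named as a Section 03 single-tentacle death animation and marked DELETE_ME, for the possessed PistolCanary TentacleBoss actor.
# A transform track holds the boss at one fixed placement, and a visibility track shows it at the start, then hides it at frame 206.
# The tentacle creature Control Rig track lists only default option channels, with no keyed control animation shown.
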LevelSequence /PistolCanary/Anim/SingleTentacle/Section03/DELETE_ME_PistolCanary_Tentacle_Death_01_1
bindings [TentacleBoss {
  id=#1
  possessed_class=PistolCanary_TentacleBoss_C
  tracks [
   {
    type=MovieSceneControlRigParameterTrack
    display_name=Tentacle_PistolCanary_Creature_CtrlRig
    sections [{
      type=MovieSceneControlRigParameterSection
      channels [
       control rig options.Location.X {type=MovieSceneScriptingFloatChannel, default=0}
       control rig options.Location.Y {type=MovieSceneScriptingFloatChannel, default=0}
       control rig options.Location.Z {type=MovieSceneScriptingFloatChannel, default=0}
       control rig options.Rotation.X {type=MovieSceneScriptingFloatChannel, default=0}
       control rig options.Rotation.Y {type=MovieSceneScriptingFloatChannel, default=0}
       control rig options.Rotation.Z {type=MovieSceneScriptingFloatChannel, default=0}
       control rig options.Scale.X {type=MovieSceneScriptingFloatChannel, default=1}
       control rig options.Scale.Y {type=MovieSceneScriptingFloatChannel, default=1}
       control rig options.Scale.Z {type=MovieSceneScriptingFloatChannel, default=1}
      ]
     }]
   }
   {
    type=MovieScene3DTransformTrack
    sections [{
      type=MovieScene3DTransformSection
      channels [
       Location.X {type=MovieSceneScriptingDoubleChannel, default=680, const [679.9999,2]}
       Location.Y {type=MovieSceneScriptingDoubleChannel, default=-6937.6663, const [-6937.6665,2]}
       Location.Z {type=MovieSceneScriptingDoubleChannel, default=448, const [448,2]}
       Rotation.X {type=MovieSceneScriptingDoubleChannel, default=0, const [0,2]}
       Rotation.Y {type=MovieSceneScriptingDoubleChannel, default=0, const [0,2]}
       Rotation.Z {type=MovieSceneScriptingDoubleChannel, default=0, const [0,2]}
       Scale.X {type=MovieSceneScriptingDoubleChannel, default=1, const [1,2]}
       Scale.Y {type=MovieSceneScriptingDoubleChannel, default=1, const [1,2]}
       Scale.Z {type=MovieSceneScriptingDoubleChannel, default=1, const [1,2]}
      ]
     }]
   }
   {
    type=MovieSceneVisibilityTrack
    sections [{type=MovieSceneVisibilitySection, channels [None {type=MovieSceneScriptingBoolChannel, default=true, keys [[0,true],[206,false]]}]}]
   }
  ]
 }]
display_rate {denominator=1, numerator=60}
tick_resolution {denominator=1, numerator=24000}
tracks []
